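# Activatable wrapper widget for text chat that switches between a message preview and the full chat panel.
# Chat buttons set the FortTextChatWrapperVM active state, while a gameplay event activates or deactivates the wrapper.
# HandleIsChatActiveChange swaps the panels, hides overlapping HUD elements, toggles an in-zone click-catcher scene and fires open dispatchers.
# Runtime-built decoy bottom-bar buttons follow input type, and the UseMessageWidgetV3 check collapses the wrapper when false.
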WidgetBlueprint /FortniteCoreUI/UI/Social/Chat/WBP_Wrapper_Text_Chat
graphs [
 BP_GetDesiredInputConfig {nodes [
   Get Desired Input Config() {is=title K2Node_FunctionEntry}
   > Return Node!(
    execute
    ReturnValue_bIgnoreMoveInput:bool=true
    ReturnValue_bIgnoreLookInput:bool=true
    ReturnValue_InputMode:ECommonInputMode Enum=All
    ReturnValue_MouseCaptureMode:EMouseCaptureMode Enum=NoCapture
    ReturnValue_MouseLockMode:EMouseLockMode Enum=DoNotLock
    ReturnValue_bHideCursorDuringViewportCapture:bool=True
   )
  ]}
 EventGraph {nodes [
   "On Button Close Clicked (WBP_TextChat)"!() -> (OutputDelegate, then, Button:Common Button Base Object Reference) {is=title K2Node_ComponentBoundEvent}
   > SetTextChatActive!(
    execute
    self:Fort Text Chat Wrapper VM Object Reference=Get FortTextChatWrapperVM() -> (FortTextChatWrapperVM:Fort Text Chat Wrapper VM Object Reference)
    bTextChatActive:bool=false
   ) -> (then)
   OnClickedChatButton!() -> (OutputDelegate, then, Button:Common Button Base Object Reference) {is=title K2Node_CustomEvent}
   > SetTextChatActive(self=Get FortTextChatWrapperVM(), bTextChatActive=true)
   n1 = SetVMChatActive!() -> (OutputDelegate, then, IsActive:bool=~) {is=title K2Node_CustomEvent}
   > SetTextChatActive(self=Get FortTextChatWrapperVM(), bTextChatActive=@n1.IsActive)
   n2 = OnInputMethodChanged!() -> (OutputDelegate, then, InputType:ECommonInputType Enum) {is=title K2Node_CustomEvent}
   > Set CurrentInputType(CurrentInputType:ECommonInputType Enum=@n2.InputType) -> (Output_Get:ECommonInputType Enum)
   > UpdateBottomBarButtons()
   Event Destruct()
   n3 = Create Event()
   n4 = Get CommonInputSubsystem()
   > Unbind Event from On Input Method Changed(self=@n4, Delegate=@n3)
   > StopListeningToEvent(self:Async Action Start Listening To Event Object Reference=Get TextChatActivationEvent() -> (TextChatActivationEvent:Async Action Start Listening To Event Object Reference))
   > SetVMChatActive!(execute, self, IsActive:bool=false) -> (then) {is=title K2Node_CallFunction}
   Event OnInitialized()
   > Assign On Active State Changed!(
    execute:Exec
    self:WBP Text Chat Object Reference=Get WBP_TextChat() -> (WBP_TextChat:WBP Text Chat Object Reference)
    Delegate:"Delegate (by ref)"=Create Event()
   ) -> (then:Exec) {is=title K2Node_AssignDelegate}
   > SetTextChatHotfix()
   Event On Activated()
   > HandleIsChatActiveChange!(execute, self, IsActivated:bool=false) -> (then)
   Event On Deactivated()
   > DeactivateWidget(self=[Get WBP_TextChat(), Get WBP_TextChat_PreviewMessages() -> (WBP_TextChat_PreviewMessages:WBP Text Chat Preview Messages Object Reference)])
   Event Construct()
   > SetIsInteractionEnabled!(
    execute
    self:Common Button Base Object Reference=Get WBP_BottomBarHackChat_Gamepad_Ref() -> (WBP_BottomBarHackChat_Gamepad_Ref:WBP Bottom Bar Decoy Button Object Reference)
    bInIsInteractionEnabled:bool=false
   ) -> (then)
   > Bind Event to On Input Method Changed(self=@n4, Delegate=@n3)
   > OnInputMethodChanged!(execute, self, InputType:ECommonInputType Enum=GetCurrentInputType(self=@n4)) -> (then) {is=title K2Node_CallFunction}
   > n5 = StartListeningToEvent!(
    execute:Exec
    Target:Gameplay Event Router Component Object Reference=GetGlobalEventRouter()
    EventType:Script Struct Object Reference=/FortniteCoreUI/UI/Social/Chat/TextChat/BPS_SetTextChatActiveState
    Context:Object Reference=GetOwningPlayer()
    EventBubblingRule:EEventBubblingRule Enum
   ) -> (
    then:Exec
    AsyncTaskProxy:Async Action Start Listening To Event Object Reference
    OnEventReceived:Exec
    Payload:BPS Set Text Chat Active State Structure
   )
   > Set TextChatActivationEvent(TextChatActivationEvent:Async Action Start Listening To Event Object Reference=@n5) -> (Output_Get:Async Action Start Listening To Event Object Reference)
   n6 = Branch(
    execute=@n5.OnEventReceived
    Condition=Break BPS Set Text Chat Active State!(BPS_SetTextChatActiveState:"BPS Set Text Chat Active State Structure (by ref)"=@n5.Payload) -> (bSetActive_5_#1:bool=~) {is=title K2Node_BreakStruct}
   )
   > ActivateWidget()
   DeactivateWidget(execute=@n6.else)
   > HandleIsChatActiveChange(IsActivated=false)
  ]}
 HandleIsChatActiveChange {nodes [
   n1 = Get WBP_TextChat_PreviewMessages()
   n2 = ActivateWidget(self=@n1)
   Call Text Chat Open Finished!(
    execute:Exec=[@n14.then, @n2.then, @n16.then, @n13.then]
    self:WBP Wrapper Text Chat Object Reference
    IsTextChatOpen:bool=Get IsActivated() -> (IsActivated:bool)
   ) -> (then:Exec) {is=title K2Node_CallDelegate}
   HandleIsChatActiveChange!() -> (then, IsActivated:bool=~) {is=title K2Node_FunctionEntry}
   > n3 = Branch(Condition=IsInZone!(self:Fort Global UIContext Object Reference=Get FortGlobalUIContext() -> (ReturnValue:Fort Global UIContext Object Reference) {is=title K2Node_GetSubsystem}) -> (ReturnValue:bool))
   > n4 = Branch(Condition=Get IsActivated())
   n5 = GetOwningPlayer()
   n6 = Get DynamicUIManager() -> (ReturnValue:Dynamic UIManager Object Reference) {is=title K2Node_GetSubsystem}
   n7 = Get DUI_Scene_ChatClickCatcherBackplate() -> (DUI_Scene_ChatClickCatcherBackplate:Dynamic UIScene Object Reference)
   > n8 = AddScene!(
    execute
    self:Dynamic UIManager Object Reference=@n6
    Scene:Dynamic UIScene Object Reference=@n7
    Player:"Player Controller Object Reference (by ref)"=@n5
   ) -> (then)
   Call Open Text Chat!(execute:Exec=[@n17.then, @n8.then, @n3.else], self:WBP Wrapper Text Chat Object Reference, IsTextChatOpen:bool=Get IsActivated()) -> (then:Exec) {is=title K2Node_CallDelegate}
   > n9 = Branch(Condition=Get IsActivated())
   n10 = Get WBP_TextChat()
   > ActivateWidget(self=@n10)
   > n11 = GetFortHUDContextByPlayer!(
    execute
    LocalPlayer:Local Player Object Reference=GetOwningLocalPlayer!(self:Widget Object Reference) -> (ReturnValue:Local Player Object Reference)
   ) -> (then, ReturnValue:Fort HUDContext Object Reference)
   > SetHUDElementVisibility!(
    execute
    self:Fort HUDContext Object Reference=@n11
    HUDElementTags:"Gameplay Tag Container Structure (by ref)"=Get HUD Tags to Hide() -> (HUD Tags to Hide:Gameplay Tag Container Structure)
    bHideElements:bool=true
    ReasonName:"Name (by ref)"=Avoiding Chat overlap
   ) -> (then)
   n12 = Get Switcher_Preview() -> (Switcher_Preview:Common Activatable Widget Switcher Object Reference)
   > SetActiveWidget(self:Widget Switcher Object Reference=@n12, Widget=Get WBP_TextChat())
   > n13 = Branch(Condition=Get bSupportOpenTextChatPreviewWindow!(self:Fort Runtime Options Object Reference=GetRuntimeOptions()) -> (bSupportOpenTextChatPreviewWindow:bool))
   n14 = DeactivateWidget(execute=@n13.else, self=@n1)
   DeactivateWidget(execute=@n9.else, self=@n10)
   > n15 = GetFortHUDContextByPlayer(LocalPlayer=GetOwningLocalPlayer())
   > SetHUDElementVisibility(self=@n15, HUDElementTags=Get HUD Tags to Hide(), bHideElements=false, ReasonName=Avoiding Chat overlap)
   > n16 = SetActiveWidget(self=@n12, Widget=Get SB_Preview() -> (SB_Preview:Size Box Object Reference))
   n17 = RemoveScene!(
    execute=@n4.else
    self:Dynamic UIManager Object Reference=@n6
    Scene:Dynamic UIScene Object Reference=@n7
    Player:"Player Controller Object Reference (by ref)"=@n5
   ) -> (then)
  ]}
 OpenTextChat {nodes [OpenTextChat!() -> (then, IsTextChatOpen:bool=~) {is=title K2Node_FunctionEntry}]}
 PopulateBottomBarButton {nodes [
   n1 = PopulateBottomBarButton!() -> (then, IsGamepad:bool=~) {is=title K2Node_FunctionEntry}
   n2 = IsTouchInputPlatform!() -> (ReturnValue:bool)
   n3 = Get Triggering Input Action_Touch() -> (Triggering Input Action_Touch:Data Table Row Handle Structure)
   n4 = Get Triggering Input Action_Gamepad() -> (Triggering Input Action_Gamepad:Data Table Row Handle Structure)
   > n5 = Construct Object from Class!(
    execute
    Class=/Game/UI/Frontend/BottomBar/WBP_BottomBarDecoyButton.WBP_BottomBarDecoyButton_C
    self
    Style:Common Button Style Class Reference=/FortUILibrary/Foundation/Buttons/ButtonStyles/Primary/ButtonStyle-Primary-M.ButtonStyle-Primary-M_C
    bApplyAlphaOnDisable:bool=True
    bLocked:bool=False
    bSelectable:bool=False
    bShouldSelectUponReceivingFocus:bool=False
    bInteractableWhenSelected:bool=False
    bToggleable:bool=False
    bTriggerClickedAfterSelection:bool=False
    bDisplayInputActionWhenNotInteractable:bool=True
    bHideInputActionWithKeyboard:bool=False
    bShouldUseFallbackDefaultInputAction:bool=True
    bRequiresHold:bool=False
    HoldData:Common UIHold Data Class Reference
    ClickMethod:EButtonClickMethod Enum=DownAndUp
    InputPriority:int=0
    TriggeringInputAction:Data Table Row Handle Structure=Select(
     Option 0:Data Table Row Handle Structure
     Option 1:Data Table Row Handle Structure=Select(Option 0=@n3, Option 1=@n4, Index=@n1)
     Index=OR Boolean(A=@n1, B=@n2)
    ) -> (ReturnValue:Data Table Row Handle Structure)
   ) -> (then, ReturnValue:WBP Bottom Bar Decoy Button Object Reference)
   > n6 = Set DecoyButton(DecoyButton:WBP Bottom Bar Decoy Button Object Reference=@n5) -> (Output_Get:WBP Bottom Bar Decoy Button Object Reference)
   > n7 = Branch(Condition=OR Boolean(A=@n1, B=IsTouchInputPlatform()))
   AddChildToHorizontalBox!(
    execute=[@n10.then, @n7.then]
    self:Horizontal Box Object Reference=Get BottomBar_Container() -> (BottomBar_Container:Horizontal Box Object Reference)
    Content:Widget Object Reference=@n6
   ) -> (then, ReturnValue:Horizontal Box Slot Object Reference)
   > n8 = Branch(Condition=@n1)
   Return Node!(
    execute=[@n9.then, @n8.then]
    BuffonRef:WBP Bottom Bar Decoy Button Object Reference=Get DecoyButton() -> (DecoyButton:WBP Bottom Bar Decoy Button Object Reference)
   )
   n9 = Bind Event to On Clicked(execute=@n8.else, self:Common Button Base Object Reference=@n6, Delegate=Create Event()) {is=title K2Node_AddDelegate}
   n10 = SetTriggeringEnhancedInputAction!(
    execute=@n7.else
    self:Common Button Base Object Reference=@n6
    InInputAction:Input Action Object Reference=/Game/Input/EnhancedInput/FortPlayerController/IA_OpenTextChat
   ) -> (then)
  ]}
 SetBottomBarButtonsVisible {nodes [
   n1 = SetBottomBarButtonsVisible!() -> (then, Visible:bool=~) {is=title K2Node_FunctionEntry}
   > Set AreBottomButtonsVisible(AreBottomButtonsVisible:bool=@n1) -> (Output_Get:bool)
   > UpdateBottomBarButtons()
  ]}
 SetTextChatHotfix {nodes [
   SetTextChatHotfix() {is=title K2Node_FunctionEntry}
   > n1 = Cast To FortPlayerController(Object=GetOwningPlayer())
   > n2 = Set IsUseMessageWidgetV3(IsUseMessageWidgetV3:bool=UseMessageWidgetV3!(self:Fort Player Controller Object Reference=@n1) -> (ReturnValue:bool)) -> (Output_Get:bool)
   > n3 = Branch(Condition=@n2)
   n4 = Get WBP_TextChat()
   > DeactivateWidget(self=@n4)
   > SetVisibility(self=@n4, InVisibility=Collapsed)
   > n5 = PopulateBottomBarButton!(execute, self, IsGamepad:bool=true) -> (then, BuffonRef:WBP Bottom Bar Decoy Button Object Reference) {is=title K2Node_CallFunction}
   > Set WBP_BottomBarHackChat_Gamepad_Ref(WBP_BottomBarHackChat_Gamepad_Ref:WBP Bottom Bar Decoy Button Object Reference=@n5) -> (Output_Get:WBP Bottom Bar Decoy Button Object Reference)
   > n6 = PopulateBottomBarButton(IsGamepad=false)
   > Set WBP_BottomBarHackChat_PC_Ref(WBP_BottomBarHackChat_PC_Ref:WBP Bottom Bar Decoy Button Object Reference=@n6) -> (Output_Get:WBP Bottom Bar Decoy Button Object Reference)
   SetVisibility(execute=@n3.else, InVisibility=Collapsed)
   > DeactivateWidget()
  ]}
 TextChatOpenFinished {nodes [TextChatOpenFinished!() -> (then, IsTextChatOpen:bool=~) {is=title K2Node_FunctionEntry}]}
 UpdateBottomBarButtons {nodes [
   UpdateBottomBarButtons() {is=title K2Node_FunctionEntry}
   n1 = Get CurrentInputType() -> (CurrentInputType:ECommonInputType Enum)
   > SetBottomButtonVisibility!(
    execute
    self:WBP Bottom Bar Decoy Button Object Reference=Get WBP_BottomBarHackChat_Gamepad_Ref()
    ShouldButtonBeVisible:bool=AND Boolean(A=Get AreBottomButtonsVisible() -> (AreBottomButtonsVisible:bool), B="Equal (Enum)"(A=Gamepad, B=@n1))
   ) -> (then)
   n2 = Get CurrentInputType()
   > SetBottomButtonVisibility(
    self=Get WBP_BottomBarHackChat_PC_Ref() -> (WBP_BottomBarHackChat_PC_Ref:WBP Bottom Bar Decoy Button Object Reference)
    ShouldButtonBeVisible=AND Boolean(A=Get AreBottomButtonsVisible(), B="Not Equal (Enum)"!(A:ECommonInputType Enum=Gamepad, B:ECommonInputType Enum=@n2) -> (ReturnValue:bool))
   )
  ]}
 UpdateIngameButtonVisibility {nodes [UpdateIngameButtonVisibility!() -> (then, ShouldBeVisible:bool=~) {is=title K2Node_FunctionEntry}]}
]
variables [
 AreBottomButtonsVisible {type=boolean}
 CurrentInputType {type=ECommonInputType, default=MouseAndKeyboard}
 DUI_Scene_ChatClickCatcherBackplate {type=DynamicUIScene, default=&/DUI_TextChatClickCatcherScene}
 HUD Tags to Hide {type=GameplayTagContainer, default {gameplayTags [{tagName=HUD.Athena.KillFeed}, {tagName=Athena.HUD.ContextualControls}]}}
 IsPreviewInstanceForMobile {type=boolean, default=true}
 IsUseMessageWidgetV3 {type=boolean, default=true}
 OpenTextChat {}
 TextChatActivationEvent {type=AsyncAction_StartListeningToEvent, default=null}
 TextChatOpenFinished {}
 Triggering Input Action_Gamepad {type=DataTableRowHandle, default {dataTable=&InputActionDataTable, rowName=ChatV2}}
 Triggering Input Action_Touch {type=DataTableRowHandle, default {dataTable=&InputActionDataTable, rowName=ChatOnlyPC}}
 WBP_BottomBarHackChat_Gamepad_Ref {type=WBP_BottomBarDecoyButton_C, default=null}
 WBP_BottomBarHackChat_PC_Ref {type=WBP_BottomBarDecoyButton_C, default=null}
]
widgets {
 info {inheritedWidgetCount=0, parentClass=&CommonActivatableWidget, rootWidgetClass=&GridPanel, widgetCount=6}
 widgets [
  GridPanel_24 {widgetClassPath=&GridPanel, widget=GridPanel_24, properties {columnFill [1], rowFill [1]}}
  BottomBar_Container {
   widgetClassPath=&HorizontalBox
   bIsVariable=true
   parent=GridPanel_24
   widget=BottomBar_Container
   slot {class=GridSlot, properties {layer=5, row=1, verticalAlignment=VAlign_Bottom}}
  }
  Switcher_Preview {
   widgetClassPath=&CommonActivatableWidgetSwitcher
   bIsVariable=true
   parent=GridPanel_24
   widget=Switcher_Preview
   properties {transitionDuration=0}
   slot {class=GridSlot, properties {layer=10}}
  }
  SB_Preview {
   widgetClassPath=&SizeBox
   bIsVariable=true
   parent=Switcher_Preview
   widget=SB_Preview
   properties {bOverride_HeightOverride=true, heightOverride=371}
   slot {class=WidgetSwitcherSlot}
  }
  WBP_TextChat_PreviewMessages {
   widgetClassPath=/FortniteCoreUI/UI/Social/Chat/TextChat/WBP_TextChat_PreviewMessages
   bIsVariable=true
   parent=SB_Preview
   widget=WBP_TextChat_PreviewMessages
   properties {bSetVisibilityOnActivated=true, bSetVisibilityOnDeactivated=true}
   slot {class=SizeBoxSlot, properties {verticalAlignment=VAlign_Bottom}}
  }
  WBP_TextChat {
   widgetClassPath=/FortniteCoreUI/UI/Social/Chat/TextChat/WBP_TextChat
   bIsVariable=true
   parent=Switcher_Preview
   widget=WBP_TextChat
   properties {bSetVisibilityOnActivated=false, bSetVisibilityOnDeactivated=false}
   slot {class=WidgetSwitcherSlot}
  }
 ]
}
