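# HUD widget that hosts the text chat wrapper and handles requests to open or close chat.
# It binds player-controller and chat events, then sets active state through the wrapper's view model, gated by a can-open flag.
# It also applies legacy-input and look settings, hands off to the main menu, and closes chat on profile-view requests before a delayed follow-up.
# On compact UI, side fillers resize open chat using safe-zone padding and a console variable.
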WidgetBlueprint /FortniteCoreUI/TextChat/UI/WBP_TextChat_InputListener
graphs [
 CloseTextChat {nodes [
   CloseTextChat() {is=title K2Node_FunctionEntry}
   > ToggleTextChat!(execute, self, ShouldOpen:bool=false) -> (then)
  ]}
 EventGraph {nodes [
   Event Construct()
   > Parent: Construct()
   n1 = Create Event()
   n2 = Get WBP_TextChat!(self:WBP Wrapper Text Chat Object Reference=Get WBP_Wrapper_Text_Chat() -> (WBP_Wrapper_Text_Chat:WBP Wrapper Text Chat Object Reference)) -> (WBP_TextChat:WBP Text Chat Object Reference)
   > Bind Event to On View Profile Requested(self:WBP Text Chat Object Reference=@n2, Delegate=@n1) {is=title K2Node_AddDelegate}
   n3 = GetOwningPlayer()
   > n4 = Cast To FortPlayerController(Object=@n3)
   n5 = Create Event()
   > Bind Event to On Text Chat Requested Delegate(self:Fort Player Controller Object Reference=@n4, Delegate=@n5) {is=title K2Node_AddDelegate}
   n6 = Create Event()
   n7 = Get WBP_TextChat(self=Get WBP_Wrapper_Text_Chat())
   > Assign On Active State Changed!(execute:Exec, self:WBP Text Chat Object Reference=@n7, Delegate:"Delegate (by ref)"=@n6) -> (then:Exec) {is=title K2Node_AssignDelegate}
   > Set IsTextChatLeftAligned(IsTextChatLeftAligned:bool=IsInZone!(self:Fort Global UIContext Object Reference=Get FortGlobalUIContext() -> (ReturnValue:Fort Global UIContext Object Reference) {is=title K2Node_GetSubsystem}) -> (ReturnValue:bool)) -> (Output_Get:bool)
   > Set MobileOpenChatSideFiller(MobileOpenChatSideFiller:Border Object Reference=Select(
     Option 0:Border Object Reference=Get Border_LeftSideFiller() -> (Border_LeftSideFiller:Border Object Reference)
     Option 1:Border Object Reference=Get Border_RightSideFiller() -> (Border_RightSideFiller:Border Object Reference)
     Index=Get IsTextChatLeftAligned() -> (IsTextChatLeftAligned:bool)
    ) -> (ReturnValue:Border Object Reference)) -> (Output_Get:Border Object Reference)
   > SetMobileLayout!(execute, self, Text Chat Open:bool=false) -> (then)
   Event OnInitialized()
   n8 = Get WBP_TextChat(self=Get WBP_Wrapper_Text_Chat())
   > SetBlockLegacyInput!(
    execute
    self:Fort Text Chat V3 Object Reference=@n8
    bBlockWhileActive:bool=Get bBlockLegacyInputWhileActive() -> (bBlockLegacyInputWhileActive:bool)
    bBlockWhileInactive:bool=Get bBlockLegacyInputWhileInactive() -> (bBlockLegacyInputWhileInactive:bool)
   ) -> (then)
   > SetEnableLookWhileActive!(execute, self:Fort Text Chat V3 Object Reference=[@n8, @n8], bEnableLook:bool=Get bAllowLookWhileActive() -> (bAllowLookWhileActive:bool)) -> (then)
   Event Destruct()
   > Parent: Destruct()
   > Unbind Event from On View Profile Requested(self:WBP Text Chat Object Reference=@n2, Delegate=@n1) {is=title K2Node_RemoveDelegate}
   > n9 = Cast To FortPlayerController(Object=@n3)
   > Unbind Event from On Text Chat Requested Delegate(self:Fort Player Controller Object Reference=@n9, Delegate=@n5) {is=title K2Node_RemoveDelegate}
   > Unbind Event from On Active State Changed(self:WBP Text Chat Object Reference=@n7, Delegate=@n6) {is=title K2Node_RemoveDelegate}
  ]}
 GetMobileOpenChatFillPercentage {nodes [
   GetMobileOpenChatFillPercentage() {is=title K2Node_FunctionEntry}
   > n1 = GetSafeZonePadding!(execute) -> (
    then
    SafePadding_X:double
    SafePadding_Y:double
    SafePadding_Z:double
    SafePadding_W:double
    SafePaddingScale:Vector 2D Structure
    SpillOverPadding:Vector 4 Structure
   )
   > n2 = GetConsoleVariableFloatValue!(execute, VariableName:String=NewTextChat.MobileOpenChatFillPercentage) -> (then, ReturnValue:float)
   n3 = float / float(A=float + float(A=@n1, B=@n1.SafePadding_Z), B=GetViewportSize!() -> (ReturnValue_X:double, ReturnValue_Y:double))
   > Return Node!(
    execute
    FillPercentage:double="Min (Float)"(A:double=float + float(A=@n2, B=@n3), B:double=1) -> (ReturnValue:double) {is=title K2Node_CommutativeAssociativeBinaryOperator}
   )
  ]}
 IsMainMenuOpen {nodes [
   IsMainMenuOpen() {is=title K2Node_FunctionEntry}
   > n1 = Cast To FortPlayerController(Object=GetOwningPlayer())
   > n2 = Cast To FortUIZone!(execute, Object=GetHUD!(self:Player Controller Object Reference=@n1) -> (ReturnValue:HUD Object Reference)) -> (then, CastFailed, AsFort UIZone:Fort UIZone Object Reference) {is=title K2Node_DynamicCast}
   > Return Node!(
    execute
    IsOpen:bool="Equal (Name)"(A=GetCurrentCursorModeAction!(self:Fort UIZone Object Reference=@n2) -> (ReturnValue:Name=None), B=ToggleMainMenu)
   )
  ]}
 IsTextChatOpen {nodes [
   IsTextChatOpen() {is=title K2Node_FunctionEntry}
   n1 = Get WBP_Wrapper_Text_Chat()
   > Return Node(IsOpen=Get bIsActive!(self:Common Activatable Widget Object Reference=Get WBP_TextChat(self=@n1)) -> (bIsActive:bool))
  ]}
 OnOpenMainMenuRequested {nodes [
   OnOpenMainMenuRequested() {is=title K2Node_FunctionEntry}
   > Branch(Condition=Get bAllowMainMenuAccess() -> (bAllowMainMenuAccess:bool))
   > OpenMainMenu()
  ]}
 OnToggleTextChatRequested {nodes [
   n1 = OnToggleTextChatRequested!() -> (then, ForceClose:bool=~) {is=title K2Node_FunctionEntry}
   n2 = IsTextChatOpen!(self) -> (IsOpen:bool) {is=title K2Node_CallFunction}
   n3 = Get WBP_TextChat_Input!(self:WBP Text Chat Object Reference=Get WBP_TextChat(self=Get WBP_Wrapper_Text_Chat())) -> (WBP_TextChat_Input:WBP Text Chat Input Object Reference)
   n4 = Get CommonInputSubsystem()
   n5 = AND Boolean!(
    A="Equal (Enum)"(A=GetCurrentInputType(self=@n4), B=MouseAndKeyboard)
    B=TextIsEmpty(InText=Get Text(self:WBP UIKit Input Field Base Object Reference=@n3))
    C:bool=IsTextChatOpen()
   ) -> (ReturnValue)
   > ToggleTextChat(ShouldOpen=NOT Boolean(A=OR Boolean!(A=@n5, B=@n1, C:bool=@n2) -> (ReturnValue)))
  ]}
 OnViewProfileRequested {nodes [
   OnViewProfileRequested() {is=title K2Node_FunctionEntry}
   > CloseTextChat() {is=title K2Node_CallFunction}
   > n1 = Branch(Condition=Get bEnableVMDrivenSidebar!(self:Fort Runtime Options Object Reference=GetRuntimeOptions()) -> (bEnableVMDrivenSidebar:bool))
   n2 = Branch(execute=@n1.else, Condition=NOT Boolean(A=IsInSubGame!(SubGame:ESubGame Enum=Campaign) -> (ReturnValue:bool)))
   > Set Timer by Event(Delegate=Create Event(), Time=0.4, bLooping=false)
   Set Timer by Event(execute=@n2.else, Delegate=Create Event(), Time=0.4, bLooping=false)
  ]}
 OpenMainMenu {nodes [
   OpenMainMenu() {is=title K2Node_FunctionEntry}
   > ToggleTextChat(ShouldOpen=false)
   n1 = GetOwningPlayer()
   > n2 = Broadcast Event!(
    execute
    self:Gameplay Event Router Component Object Reference=GetGlobalEventRouter()
    EventData:Toggle Main Menu Event Structure=Make Toggle Main Menu Event!(ShouldOpen_1_#1:bool=true) -> (ToggleMainMenuEvent:Toggle Main Menu Event Structure) {is=title K2Node_MakeStruct}
    EventContext:Object Reference=@n1
   ) -> (then, ReturnValue:bool)
   > n3 = Branch(Condition=@n2)
   n4 = Cast To FortPlayerController(execute=@n3.else, Object=@n1)
   > ToggleMainMenu(self:Fort Player Controller Object Reference=@n4)
  ]}
 SetBlocksLegacyInput {nodes [
   n1 = SetBlocksLegacyInput!() -> (then, BlockWhileActive:bool=~, BlockWhileInactive:bool=~) {is=title K2Node_FunctionEntry}
   > Set bBlockLegacyInputWhileActive(bBlockLegacyInputWhileActive:bool=@n1) -> (Output_Get:bool)
   > Set bBlockLegacyInputWhileInactive(bBlockLegacyInputWhileInactive:bool=@n1.BlockWhileInactive) -> (Output_Get:bool)
   > SetBlockLegacyInput(
    self=Get WBP_TextChat(self=Get WBP_Wrapper_Text_Chat())
    bBlockWhileActive=Get bBlockLegacyInputWhileActive()
    bBlockWhileInactive=Get bBlockLegacyInputWhileInactive()
   )
  ]}
 SetCanOpenChat {nodes [
   n1 = SetCanOpenChat!() -> (then, bCanOpenChat:bool=~) {is=title K2Node_FunctionEntry}
   > Set bCanOpenChat(bCanOpenChat:bool=@n1) -> (Output_Get:bool)
  ]}
 SetMobileLayout {nodes [
   SetMobileLayout!() -> (then, Text Chat Open:bool=~) {is=title K2Node_FunctionEntry}
   > Branch(Condition=IsUsingCompactUI())
   > n1 = GetMobileOpenChatFillPercentage!(execute, self) -> (then, FillPercentage:double) {is=title K2Node_CallFunction}
   > Set MobileOpenChatFillPercentage(MobileOpenChatFillPercentage:double=@n1) -> (Output_Get:double)
   n2 = SlotAsHorizontalBoxSlot!(Widget:Widget Object Reference=Get WBP_Wrapper_Text_Chat()) -> (ReturnValue:Horizontal Box Slot Object Reference)
   > SetHorizontalAlignment(
    self:Horizontal Box Slot Object Reference=@n2
    InHorizontalAlignment=Select(
     Option 0:EHorizontalAlignment Enum=Get MobileClosedChatHorizontalAligment() -> (MobileClosedChatHorizontalAligment:EHorizontalAlignment Enum=HAlign_Fill)
     Option 1:EHorizontalAlignment Enum=HAlign_Fill
     Index=Get Text Chat Open() -> (Text Chat Open:bool)
    ) -> (ReturnValue:EHorizontalAlignment Enum)
   )
   > SetVerticalAlignment(
    self:Horizontal Box Slot Object Reference=@n2
    InVerticalAlignment=Select(
     Option 0:EVerticalAlignment Enum=Get MobileClosedChatVerticalAligment() -> (MobileClosedChatVerticalAligment:EVerticalAlignment Enum=VAlign_Fill)
     Option 1:EVerticalAlignment Enum=VAlign_Fill
     Index=Get Text Chat Open()
    ) -> (ReturnValue:EVerticalAlignment Enum)
   )
   n3 = Get MobileOpenChatFillPercentage() -> (MobileOpenChatFillPercentage:double)
   n4 = Get Text Chat Open()
   > SetSize!(
    execute
    self:Horizontal Box Slot Object Reference=@n2
    InSize:Slate Child Size Structure=Make Slate Child Size!(Value:float=SelectFloat(A=@n3, B=1, bPickA=@n4), SizeRule:ESlateSizeRule Enum=Fill) -> (SlateChildSize:Slate Child Size Structure) {is=title K2Node_MakeStruct}
   ) -> (then)
   > SetPadding!(
    execute
    self:Horizontal Box Slot Object Reference=@n2
    InPadding:Margin Structure=Select(
     Option 0:Margin Structure=Get MobileClosedChatPadding() -> (MobileClosedChatPadding:Margin Structure)
     Option 1:Margin Structure
     Index=Get Text Chat Open()
    ) -> (ReturnValue:Margin Structure)
   ) -> (then)
   n5 = float - float(A=1, B=Get MobileOpenChatFillPercentage())
   n6 = Get Text Chat Open()
   > SetSize(
    self=SlotAsHorizontalBoxSlot(Widget=Get MobileOpenChatSideFiller() -> (MobileOpenChatSideFiller:Border Object Reference))
    InSize=Make Slate Child Size(Value=SelectFloat(A=@n5, B=0, bPickA=@n6), SizeRule=Fill)
   )
  ]}
 ToggleTextChat {nodes [
   n1 = ToggleTextChat!() -> (then, ShouldOpen:bool=~) {is=title K2Node_FunctionEntry}
   > Branch(Condition=OR Boolean!(A=NOT Boolean(A=@n1), B=Get bCanOpenChat() -> (bCanOpenChat:bool)) -> (ReturnValue))
   > Branch(Condition="Not Equal (Boolean)"(A=@n1, B=IsTextChatOpen()) -> (ReturnValue=~))
   > SetTextChatActive!(
    execute
    self:Fort Text Chat Wrapper VM Object Reference=Get FortTextChatWrapperVM!(self:WBP Wrapper Text Chat Object Reference=Get WBP_Wrapper_Text_Chat()) -> (FortTextChatWrapperVM:Fort Text Chat Wrapper VM Object Reference)
    bTextChatActive:bool=@n1
   ) -> (then)
   > SetMobileLayout(Text Chat Open=@n1) {is=title K2Node_CallFunction}
  ]}
 TryOpeningSocialPanel {nodes [
   TryOpeningSocialPanel() {is=title K2Node_FunctionEntry}
   > Broadcast Event(
    self=GetGlobalEventRouter()
    EventData:Sidebar Request Open Structure=Make Sidebar Request Open!() -> (SidebarRequestOpen:Sidebar Request Open Structure) {is=title K2Node_MakeStruct}
    EventContext=GetOwningPlayer()
   )
  ]}
]
variables [
 bAllowLookWhileActive {type=boolean, default=true}
 bAllowMainMenuAccess {type=boolean, default=true}
 bBlockLegacyInputWhileActive {type=boolean}
 bBlockLegacyInputWhileInactive {type=boolean}
 bCanOpenChat {type=boolean, default=true}
 IsTextChatLeftAligned {type=boolean}
 MobileClosedChatHorizontalAligment {type=EHorizontalAlignment, default=HAlign_Left}
 MobileClosedChatPadding {type=Margin, default=<ltrb 0 0 0 0>}
 MobileClosedChatVerticalAligment {type=EVerticalAlignment, default=VAlign_Bottom}
 MobileOpenChatSideFiller {type=Border, default=null}
]
widgets {
 info {inheritedWidgetCount=1, parentClass=&FortHUDElementWidget, rootWidgetClass=&GridPanel, widgetCount=6}
 widgets [
  GridPanel_0 {widgetClassPath=&GridPanel, widget=GridPanel_0, properties {columnFill [1], rowFill [1]}}
  HorizontalBox {widgetClassPath=&HorizontalBox, bIsVariable=true, parent=GridPanel_0, widget=HorizontalBox, slot {class=GridSlot}}
  Border_LeftSideFiller {
   widgetClassPath=&Border
   bIsVariable=true
   parent=HorizontalBox
   widget=Border_LeftSideFiller
   properties {visibility=Hidden, brushColor=<rgba 1 1 1 0>, padding=<ltrb 0 0 0 0>}
   slot {class=HorizontalBoxSlot, properties {size {sizeRule=Fill, value=0}}}
  }
  WBP_Wrapper_Text_Chat {
   widgetClassPath=/FortniteCoreUI/UI/Social/Chat/WBP_Wrapper_Text_Chat
   bIsVariable=true
   parent=HorizontalBox
   widget=WBP_Wrapper_Text_Chat
   slot {class=HorizontalBoxSlot, properties {horizontalAlignment=HAlign_Left, verticalAlignment=VAlign_Bottom, size {sizeRule=Fill, value=1}}}
  }
  Border_RightSideFiller {
   widgetClassPath=&Border
   bIsVariable=true
   parent=HorizontalBox
   widget=Border_RightSideFiller
   properties {visibility=Hidden, brushColor=<rgba 1 1 1 0>, padding=<ltrb 0 0 0 0>}
   slot {class=HorizontalBoxSlot, properties {size {sizeRule=Fill, value=0}}}
  }
  TutorialHighlight {widgetClassPath=&TutorialHighlight}
 ]
}
